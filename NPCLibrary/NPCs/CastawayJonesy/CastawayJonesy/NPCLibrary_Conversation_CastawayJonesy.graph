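# Conversation database for the Castaway Jonesy NPC, defining the greeting shown when a player talks to the character.
# A single entry point leads to a Speech node with a default line plus contextual variants.
# One variant targets first-ever conversations, while the others give high-priority, high-weight reactions to specific player skins through character ID requirements.
# The speech then branches through link nodes into the shared Season 16 Spire Joneses, BasicServices and NPCLibrary Quests conversations.
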ConversationDatabase /NPCLibrary/NPCs/CastawayJonesy/CastawayJonesy/NPCLibrary_Conversation_CastawayJonesy
entry_points [{tag=Conversation.Entry.CastawayJonesy, destinations [#1]}]
nodes [
 ConversationEntryPointNode_7 {
  id=#1
  class=ConversationEntryPointNode
  outputs [#2]
  properties {entryTag {tagName=Conversation.Entry.CastawayJonesy}}
 }
 Speech_C_4 {
  id=#2
  class=Speech_C
  outputs [#3, #4, #5]
  properties {
   nodeName=""
   generalConfig {
    defaultMessage="Any coconuts in the fridge with my name on them are MINE. Wait. \"Jones\"... aw, nuts..."
    contextualMessages [
     {
      message="This place has FOOD! And walls! And... people. Yeah. Great."
      contextRequirements [{requirement {class=FortControllerRequirement_IsFirstEverConversationWithNPC}}]
      priority {curve {rowName=Tandem.Speech.Priority.FirstEverInteraction}}
     }
     {
      message=Well if it isn't Mr. I Have Multiple Outfits.
      contextRequirements [{requirement {
         class=FortControllerRequirement_HasCID
         properties {allowedCIds [&CID_989_Athena_Commando_M_ProgressiveJonesy]}
        }}]
      priority {value=10, curve {curveTable=null, rowName=None}}
      weight {value=10}
     }
     {
      message=An island is TOTALLY worse than a bunker.
      contextRequirements [{requirement {
         class=FortControllerRequirement_HasCID
         properties {allowedCIds [/BRCosmetics/Athena/Items/Cosmetics/Characters/CID_409_Athena_Commando_M_BunkerMan]}
        }}]
      priority {value=10, curve {curveTable=null, rowName=None}}
      weight {value=10}
     }
     {
      message="Oh, the sun got you pretty bad too, huh."
      contextRequirements [{requirement {
         class=FortControllerRequirement_HasCID
         properties {allowedCIds [/BRCosmetics/Athena/Items/Cosmetics/Characters/CID_836_Athena_Commando_M_JonesyFlare]}
        }}]
      priority {value=10, curve {curveTable=null, rowName=None}}
      weight {value=10}
     }
     {
      message=Did you make coffee mugs out of my coconuts??
      contextRequirements [{requirement {
         class=FortControllerRequirement_HasCID
         properties {allowedCIds [/BRCosmetics/Athena/Items/Cosmetics/Characters/CID_798_Athena_Commando_M_JonesyVagabond]}
        }}]
      priority {value=10, curve {curveTable=null, rowName=None}}
      weight {value=10}
     }
     {
      message=Did I really have that haircut?
      contextRequirements [{requirement {
         class=FortControllerRequirement_HasCID
         properties {allowedCIds [&CID_883_Athena_Commando_M_ChOneJonesy]}
        }}]
      priority {value=10, curve {curveTable=null, rowName=None}}
      weight {value=10}
     }
     {
      message="On my island, I didn't see anyone for years. It was GREAT!"
      contextRequirements [{requirement {class=FortControllerRequirement_HasCID, properties {allowedCIds [&Character_ChillCat]}}}]
      priority {value=10, curve {curveTable=null, rowName=None}}
      weight {value=10}
     }
     {
      message="Do those hands sunburn, or is that like, not a problem for you?"
      contextRequirements [{requirement {class=FortControllerRequirement_HasCID, properties {allowedCIds [&Character_Bites]}}}]
      priority {value=10, curve {curveTable=null, rowName=None}}
      weight {value=10}
     }
     {
      message=I know what it's like to be trapped somewhere.
      contextRequirements [{requirement {class=FortControllerRequirement_HasCID, properties {allowedCIds [&Character_PinkSpike]}}}]
      priority {value=10, curve {curveTable=null, rowName=None}}
      weight {value=10}
     }
     {
      message="I know bears like honey, but do they like coconuts?"
      contextRequirements [{requirement {class=FortControllerRequirement_HasCID, properties {allowedCIds [&Character_BadBear]}}}]
      priority {value=10, curve {curveTable=null, rowName=None}}
      weight {value=10}
     }
     {
      message=I'm watching you.
      contextRequirements [{requirement {class=FortControllerRequirement_HasCID, properties {allowedCIds [&Character_TheHerald]}}}]
      priority {value=10, curve {curveTable=null, rowName=None}}
      weight {value=10}
     }
     {
      message="How long was I on that island? Spiders are people now?!"
      contextRequirements [{requirement {class=FortControllerRequirement_HasCID, properties {allowedCIds [&Character_RoseDust]}}}]
      priority {value=10, curve {curveTable=null, rowName=None}}
      weight {value=10}
     }
    ]
   }
  }
 }
 ConversationLinkNode_1 {id=#5, class=ConversationLinkNode, properties {remoteEntryTag {tagName=Conversation.Entry.NPCLibrary.Quests}}}
 ConversationLinkNode_0 {id=#4, class=ConversationLinkNode, properties {remoteEntryTag {tagName=Conversation.Entry.BasicServices}}}
 ConversationLinkNode_2 {
  id=#3
  class=ConversationLinkNode
  properties {remoteEntryTag {tagName=Conversation.Entry.Season16.TheSpire.Joneses}}
 }
]
